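# Niagara system for the Bass Burn emote that spawns a burst of gas fire with embers, smoke and light.
# Six independent burst emitters layer flame meshes, short and tall fire-lick sprites, GPU curl-noise embers, rising smoke and a fading light that jitters toward the owner.
# Fire and smoke spawn on flattened hemispheres scattered along X, animated through curve-driven material parameters.
# Embers, licks and smoke use distance culling; continuous spawn-rate modules are disabled.
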NiagaraSystem /SparksCosmetics/Emotes/Bass/BassBurn/Effects/NS_Emote_BassBurn_GasFire_Burst
dependencies {
 usedDataInterfaces [&NiagaraDataInterfaceColorCurve, &NiagaraDataInterfaceCurve, &NiagaraDataInterfaceVector2DCurve]
 usedDynamicInputs [&MakeFloatFromVector, &MakeFloatFromVector4]
 usedModules [&SystemState]
}
emitters {
 Embers_Burst {
  simTarget=GPUComputeSim
  emitterUpdateScript [
   &EmitterState(
    Enable Distance Culling=true
    Enable Visibility Culling=true
    MaxDistance=1000
    Module.bApplySpawnCountScaleByDistance=true
    Module.bUseMaxDistance=true
    Module.bUseMinDistance=false
    Scalability Mode=::NewEnumerator1
    Scale Spawn Count=true
    Spawn Count Scale By Distance=DataInterface() {propertyValues {Curve {keys [<at 0 1>,<at 0.3 1>,<at 1 0>]}}}
   )
   &SpawnBurst_Instantaneous(Spawn Count=25, Use Spawn Probability=false, UseLoopCountLimit=false)
  ]
  particleSpawnScript [
   &InitializeParticle(
    Lifetime=&UniformRangedFloat(Minimum=0.3, Maximum=1)
    Mass Mode=::NewEnumerator1
    Position=&SimulationPosition()
    Position Mode=::NewEnumerator1
    Sprite Size=&Vector2DFromFloat(Value=&FloatFromCurve(
      FloatCurve=DataInterface() {propertyValues {Curve {keys [<at 0 1.2>,<at 0.1257 0.9847>,<at 0.2996 0.5>,<at 1 0.3>]}}}
      CurveIndex=&UniformRangedFloat(Minimum=0, Maximum=1)
      Scale Curve=2
     ))
    Sprite Size Mode=::NewEnumerator4
    UsePositionOffset=false
    Write Material Random=true
   ) {module=InitializeParticle001}
   &SphereLocation(
    Hemisphere Z=true
    MaskSpawnGroup=false
    Non Uniform Scale=<1 1 0.5>
    Offset=&RandomRangeVector(Minimum=<-45 -5 -15>, Maximum=<50 5 -5>)
    Override Seed=false
    Sphere Radius=3
    Surface Only=true
    UseNonUniformScale=true
    UseOffset=true
    UseSphereOrientation=false
   )
   &AddVelocity(
    Cone Angle=90
    Cone Axis=<1 0 1>
    Module.UseConeFalloff=false
    Module.UseVelDistribution=false
    Override Seed=false
    Velocity Mode=::NewEnumerator2
    Velocity Speed=&RandomRangeFloat(Minimum=50, Maximum=250)
   )
  ]
  particleUpdateScript [
   &ParticleState()
   &Drag(
    Drag=&FloatFromCurve(FloatCurve=DataInterface() {propertyValues {Curve {keys [<at 0 0.5>,<at 0.5 0.5>,<at 1 1>]}}}, Scale Curve=9)
    Use Linear Drag=true
    Use Rotational Drag=false
   )
   /Niagara/Modules/Update/Forces/CurlNoiseForce(
    Noise Frequency=0.04
    Noise Strength=&FloatFromCurve(FloatCurve=DataInterface() {propertyValues {Curve {keys [<at 0 1>,<at 1 0>]}}}, Scale Curve=400)
    Pan Noise=true
    Pan Noise Field=<0 0 1>
   )
   &GravityForce(Gravity=&RandomRangeVector(Minimum=<0 0 500>, Maximum=<0 0 700>))
   &SolveForcesAndVelocity(Clamp Velocity=false, Limit Acceleration=false)
   &ScaleSpriteSizeBySpeed(
    Max Scale Factor=<2 5>
    Sample Scale Factor By Curve=true
    Scale Factor Curve=DataInterface() {propertyValues {Curve {keys [
        <at -0.0077 0.0765 RCIM_Cubic>
        {arriveTangent=1.2463, interpMode=RCIM_Cubic, leaveTangent=1.2463, tangentMode=RCTM_User, time=0.9985, value=0.9549}
       ]}}}
    Velocity Threshold=500
   )
   &ScaleColor(
    Scale Alpha=&FloatFromCurve(FloatCurve=DataInterface() {propertyValues {Curve {keys [<at 0 1>,<at 1 0>]}}})
    Scale RGB=&MakeVectorFromLinearColorRGB(Color=&ColorFromCurve(ColorCurve=DataInterface() {propertyValues {
        AlphaCurve {keys [<at 0 1 RCIM_Cubic>,<at 0.7107 1>]}
        BlueCurve {keys [<at 0 14.6701 RCIM_Cubic>,<at 0.2499 1.5104>,<at 0.7114 0.1753 RCIM_Cubic>]}
        GreenCurve {keys [<at 0 36.7481 RCIM_Cubic>,<at 0.2499 5.293>,<at 0.7114 0.3243 RCIM_Cubic>]}
        RedCurve {keys [<at 0 50 RCIM_Cubic>,<at 0.2499 10>,<at 0.7114 1 RCIM_Cubic>]}
       }}))
    ScaleA=true
    ScaleRGB=true
   )
  ]
  rendererClasses [&NiagaraSpriteRendererProperties]
  renderers [{rendererClass=&NiagaraSpriteRendererProperties}]
 }
 FireLicks_Sprites {
  emitterUpdateScript [
   &EmitterState(Enable Distance Culling=true, MaxDistance=3000, Module.bUseMaxDistance=true, Module.bUseMinDistance=false, Scalability Mode=::NewEnumerator1)
   &SpawnBurst_Instantaneous(Spawn Count=25, Use Spawn Probability=false, UseLoopCountLimit=false)
   &SpawnRate(SpawnRate=50, Use Spawn Probability=false) {enabled=false, module=SpawnRate001}
   &SpawnRate(
    SpawnRate=/Niagara/DynamicInputs/Math/LerpMultipleFloats(
     Float Count:ENiagara_MultipleLerpCount=::NewEnumerator0
     1:NiagaraFloat=&UniformRangedFloat(Minimum=15, Maximum=20)
     2:NiagaraFloat=6
     3:NiagaraFloat=0
     4:NiagaraFloat=?
     5:NiagaraFloat=?
     Alpha:NiagaraFloat=&ClampFloat(Float=&FN_LOD_Distance_Remap(Start Distance=600, End Distance=2400, Minimum Output=0, Maximum Output=1))
    )
    Use Spawn Probability=false
   ) {enabled=false}
  ]
  particleSpawnScript [
   &InitializeParticle(
    Lifetime=&Multiply_Float(A=&UniformRangedFloat(Minimum=1, Maximum=2), B=0.5)
    Mass Mode=::NewEnumerator1
    Sprite Rotation Angle=&UniformRangedFloat(Minimum=0, Maximum=360)
    Sprite Rotation Mode=::NewEnumerator0
    Sprite Size=&Vector2DFromFloat(Value=&Multiply_Float(A=0.4, B=&UniformRangedFloat(Minimum=110, Maximum=145)))
    Sprite Size Mode=::NewEnumerator4
    UsePositionOffset=false
    Write Material Random=true
   ) {module=InitializeParticle001}
   &Color(Color=<rgba 10 0.3707 0 0.55>) {enabled=false}
   &Color(Color=&ColorFromCurve(ColorCurve=DataInterface() {propertyValues {
       AlphaCurve {keys [<at 0 1>]}
       BlueCurve {keys [
         {arriveTangent=0.0977, interpMode=RCIM_Cubic, time=0.2155, value=0.0224}
         {arriveTangent=0.313, interpMode=RCIM_Cubic, leaveTangent=0.313, tangentMode=RCTM_User, time=0.7941, value=0.0776}
         {arriveTangent=1.6688, interpMode=RCIM_Cubic, leaveTangent=1.6688, tangentMode=RCTM_User, time=0.8913, value=0.25}
        ]}
       GreenCurve {keys [
         {arriveTangent=0.2525, interpMode=RCIM_Cubic, leaveTangent=0.2525, tangentMode=RCTM_User, time=0.2155, value=0.0633}
         {arriveTangent=1.1982, interpMode=RCIM_Cubic, leaveTangent=1.1982, tangentMode=RCTM_User, time=0.8, value=0.35}
         {arriveTangent=2.4333, interpMode=RCIM_Cubic, leaveTangent=2.4333, tangentMode=RCTM_User, time=0.8913, value=0.7}
        ]}
       RedCurve {keys [
         {arriveTangent=1.309, interpMode=RCIM_Cubic, leaveTangent=1.309, tangentMode=RCTM_User, time=0.2155, value=0.45}
         {arriveTangent=0.8354, interpMode=RCIM_Cubic, leaveTangent=0.8354, tangentMode=RCTM_User, time=0.8, value=0.9}
         {arriveTangent=0.0631, interpMode=RCIM_Cubic, leaveTangent=0.0631, tangentMode=RCTM_User, time=0.8913, value=1}
        ]}
      }})) {module=Color001}
   &DynamicMaterialParameters(
    Index 0 Param 2=&RandomRangeFloat(Minimum=0, Maximum=4)
    Index 0 Param 4=2.5
    Module.Param0WriteEnabled=false
    Module.Param1WriteEnabled=true
    Module.Param2WriteEnabled=false
    Module.Param3WriteEnabled=true
   ) {module=DynamicMaterialParameters001}
   &SphereLocation(
    Hemisphere Z=true
    MaskSpawnGroup=false
    Non Uniform Scale=<1 1 0.5>
    Offset=&RandomRangeVector(Minimum=<-45 -5 -15>, Maximum=<50 5 -5>)
    Override Seed=false
    Sphere Radius=16
    Surface Only=true
    UseNonUniformScale=true
    UseOffset=true
    UseSphereOrientation=false
   ) {module=SphereLocation001}
   &SphereLocation(
    Hemisphere Z=true
    MaskSpawnGroup=false
    Non Uniform Scale=<1 1 0.5>
    Override Seed=false
    Sphere Radius=16
    Surface Only=true
    UseNonUniformScale=true
    UseOffset=true
    UseSphereOrientation=false
   ) {enabled=false}
  ]
  particleUpdateScript [
   /Niagara/Modules/Update/SubUV/SubUVAnimation(Number Of Frames=&FloatFromCurve(FloatCurve=DataInterface() {propertyValues {Curve {keys [{arriveTangent=5.9363, interpMode=RCIM_Cubic, leaveTangent=5.9361, tangentMode=RCTM_User, time=0, value=56}, <at 0.9 60>]}}}))
   &ParticleState()
   &ScaleColor(
    Scale Alpha=&Multiply_Float(
     A=0.8
     B=&FloatFromCurve(FloatCurve=DataInterface() {propertyValues {Curve {keys [
          <at 0 0 RCIM_Cubic>
          {arriveTangent=-0.0116, interpMode=RCIM_Cubic, leaveTangent=-0.0116, tangentMode=RCTM_User, time=0.4073, value=1}
          {arriveTangent=-2.0743, interpMode=RCIM_Cubic, leaveTangent=-2.0743, tangentMode=RCTM_User, time=0.6746, value=0.6742}
          {interpMode=RCIM_Cubic, leaveTangent=-2.371, time=1, value=0}
         ]}}})
    )
    Scale RGB=&VectorFromFloat(Value=&FloatFromCurve(
      FloatCurve=DataInterface() {propertyValues {Curve {keys [<at 0 1 RCIM_Cubic>, {arriveTangent=-2.3885, tangentMode=RCTM_User, time=1, value=0}]}}}
      Scale Curve=10
     ))
    ScaleA=true
    ScaleRGB=true
   ) {module=ScaleColor001}
   &ScaleSpriteSize(Scale Factor=&Vector2DFromCurve(
     Vector2Curve=DataInterface() {propertyValues {XCurve {keys [<at 0 0.6>,<at 1 1>]}, YCurve {keys [<at 0 0.6>,<at 1 1>]}}}
     ScaleCurve=&Vector2DFromFloat(Value=&FN_LOD_Distance_Remap(Start Distance=300, End Distance=1000, Minimum Output=1, Maximum Output=1.1))
    ))
   &DynamicMaterialParameters(
    Index 0 Param 1=&FloatFromCurve(FloatCurve=DataInterface() {propertyValues {Curve {keys [<at 0 1>]}}})
    Index 0 Param 2=0.5
    Index 0 Param 3=&FloatFromCurve(FloatCurve=DataInterface() {propertyValues {Curve {keys [<at 0 0.5 RCIM_Cubic>, {arriveTangent=1.5, interpMode=RCIM_Cubic, leaveTangent=1.5, time=0.5235, value=0.5}, <at 1 2>]}}})
    Module.Param0WriteEnabled=true
    Module.Param1WriteEnabled=false
    Module.Param2WriteEnabled=true
    Module.Param3WriteEnabled=false
   )
   &SpriteRotationRate(Rotation Rate=&Lerp_Float(A=-10, B=10, Alpha=$Particles.Initial.MaterialRandom)) {module=SpriteRotationRate001}
   &GravityForce(Gravity=&RandomRangeVector(Minimum=<0 0 50>, Maximum=<0 0 250>))
   &SolveForcesAndVelocity(Clamp Velocity=false, Limit Acceleration=false)
   &CameraOffset(Camera Offset Amount=-2)
  ]
  rendererClasses [&NiagaraSpriteRendererProperties]
  renderers [{rendererClass=&NiagaraSpriteRendererProperties}]
 }
 FireLicks_Sprites_Tall {
  emitterUpdateScript [
   &EmitterState(Enable Distance Culling=true, MaxDistance=3000, Module.bUseMaxDistance=true, Module.bUseMinDistance=false, Scalability Mode=::NewEnumerator1)
   &SpawnBurst_Instantaneous(Spawn Count=10, Use Spawn Probability=false, UseLoopCountLimit=false)
  ]
  particleSpawnScript [
   &InitializeParticle(
    Lifetime=&Multiply_Float(A=&UniformRangedFloat(Minimum=0.5, Maximum=2), B=0.5)
    Mass Mode=::NewEnumerator1
    Sprite Rotation Mode=::NewEnumerator2
    Sprite Size=&Vector2DFromFloat(Value=&Multiply_Float(A=1, B=&UniformRangedFloat(Minimum=110, Maximum=145)))
    Sprite Size Mode=::NewEnumerator4
    UsePositionOffset=false
    Write Material Random=true
   ) {module=InitializeParticle001}
   &Color(Color=<rgba 10 0.3707 0 0.55>) {enabled=false}
   &Color(Color=&ColorFromCurve(ColorCurve=DataInterface() {propertyValues {
       AlphaCurve {keys [<at 0 1>]}
       BlueCurve {keys [
         {arriveTangent=0.0977, interpMode=RCIM_Cubic, time=0.2155, value=0.0224}
         {arriveTangent=0.313, interpMode=RCIM_Cubic, leaveTangent=0.313, tangentMode=RCTM_User, time=0.7941, value=0.0776}
         {arriveTangent=1.6688, interpMode=RCIM_Cubic, leaveTangent=1.6688, tangentMode=RCTM_User, time=0.8913, value=0.25}
        ]}
       GreenCurve {keys [
         {arriveTangent=0.2525, interpMode=RCIM_Cubic, leaveTangent=0.2525, tangentMode=RCTM_User, time=0.2155, value=0.0633}
         {arriveTangent=1.1982, interpMode=RCIM_Cubic, leaveTangent=1.1982, tangentMode=RCTM_User, time=0.8, value=0.35}
         {arriveTangent=2.4333, interpMode=RCIM_Cubic, leaveTangent=2.4333, tangentMode=RCTM_User, time=0.8913, value=0.7}
        ]}
       RedCurve {keys [
         {arriveTangent=1.309, interpMode=RCIM_Cubic, leaveTangent=1.309, tangentMode=RCTM_User, time=0.2155, value=0.45}
         {arriveTangent=0.8354, interpMode=RCIM_Cubic, leaveTangent=0.8354, tangentMode=RCTM_User, time=0.8, value=0.9}
         {arriveTangent=0.0631, interpMode=RCIM_Cubic, leaveTangent=0.0631, tangentMode=RCTM_User, time=0.8913, value=1}
        ]}
      }})) {module=Color001}
   &DynamicMaterialParameters(
    Index 0 Param 2=&RandomRangeFloat(Minimum=0, Maximum=4)
    Index 0 Param 4=2.5
    Module.Param0WriteEnabled=false
    Module.Param1WriteEnabled=true
    Module.Param2WriteEnabled=false
    Module.Param3WriteEnabled=true
   ) {module=DynamicMaterialParameters001}
   &SphereLocation(
    Hemisphere Z=true
    MaskSpawnGroup=false
    Non Uniform Scale=<1 1 0.5>
    Offset=&RandomRangeVector(Minimum=<0 -5 -15>, Maximum=<50 5 -5>)
    Override Seed=false
    Sphere Radius=16
    Surface Only=true
    UseNonUniformScale=true
    UseOffset=true
    UseSphereOrientation=false
   ) {module=SphereLocation001}
   &AddVelocity(
    Cone Axis=<0 0 1>
    Module.UseConeFalloff=false
    Module.UseVelDistribution=false
    Override Seed=false
    Velocity Mode=::NewEnumerator2
    Velocity Speed=&RandomRangeFloat(Minimum=50, Maximum=175)
   )
  ]
  particleUpdateScript [
   &SubUVAnimation(Number Of Frames=&FloatFromCurve(FloatCurve=DataInterface() {propertyValues {Curve {keys [{arriveTangent=5.9363, interpMode=RCIM_Cubic, leaveTangent=5.9361, tangentMode=RCTM_User, time=0, value=56}, <at 0.9 60>]}}}))
   &ParticleState()
   &ScaleColor(
    Scale Alpha=&Multiply_Float(
     A=0.8
     B=&FloatFromCurve(FloatCurve=DataInterface() {propertyValues {Curve {keys [
          {arriveTangent=9.0309, interpMode=RCIM_Cubic, leaveTangent=9.0309, tangentMode=RCTM_User, time=0, value=0}
          {arriveTangent=-0.0116, interpMode=RCIM_Cubic, leaveTangent=-0.0116, tangentMode=RCTM_User, time=0.4073, value=1}
          {arriveTangent=-2.0743, interpMode=RCIM_Cubic, leaveTangent=-2.0743, tangentMode=RCTM_User, time=0.8, value=0.7}
          {interpMode=RCIM_Cubic, leaveTangent=-2.371, time=1, value=0}
         ]}}})
    )
    Scale RGB=&VectorFromFloat(Value=&FloatFromCurve(
      FloatCurve=DataInterface() {propertyValues {Curve {keys [<at 0 1 RCIM_Cubic>, {arriveTangent=-2.3885, tangentMode=RCTM_User, time=1, value=0}]}}}
      Scale Curve=10
     ))
    ScaleA=true
    ScaleRGB=true
   ) {module=ScaleColor001}
   &ScaleSpriteSize(Scale Factor=&Vector2DFromCurve(
     Vector2Curve=DataInterface() {propertyValues {XCurve {keys [<at 0 0.6>,<at 1 1>]}, YCurve {keys [<at 0 0.6>,<at 1 1>]}}}
     ScaleCurve=&Vector2DFromFloat(Value=&FN_LOD_Distance_Remap(Start Distance=300, End Distance=1000, Minimum Output=1, Maximum Output=1.1))
    ))
   &DynamicMaterialParameters(
    Index 0 Param 1=&FloatFromCurve(FloatCurve=DataInterface() {propertyValues {Curve {keys [<at 0 1>]}}})
    Index 0 Param 2=0.5
    Index 0 Param 3=&FloatFromCurve(FloatCurve=DataInterface() {propertyValues {Curve {keys [
         <at 0 0.5 RCIM_Cubic>
         {arriveTangent=-0.0025, interpMode=RCIM_Cubic, leaveTangent=-0.0025, tangentMode=RCTM_User, time=0.6, value=0.625}
         {arriveTangent=2.7303, tangentMode=RCTM_User, time=1, value=2}
        ]}}})
    Module.Param0WriteEnabled=true
    Module.Param1WriteEnabled=false
    Module.Param2WriteEnabled=true
    Module.Param3WriteEnabled=false
   )
   &SpriteRotationRate(Rotation Rate=&Lerp_Float(A=-10, B=10, Alpha=$Particles.Initial.MaterialRandom)) {module=SpriteRotationRate001}
   &GravityForce(Gravity=&RandomRangeVector(Minimum=<0 0 125>, Maximum=<0 0 275>))
   &Drag(Use Linear Drag=true, Use Rotational Drag=true)
   &SolveForcesAndVelocity(Clamp Velocity=false, Limit Acceleration=false)
   &CameraOffset(Camera Offset Amount=-2)
  ]
  rendererClasses [&NiagaraSpriteRendererProperties]
  renderers [{rendererClass=&NiagaraSpriteRendererProperties}]
 }
 Fire_Main {
  emitterUpdateScript [
   &EmitterState()
   &SpawnBurst_Instantaneous(Spawn Count=45, Use Spawn Probability=false, UseLoopCountLimit=false)
   &SpawnRate(SpawnRate=45, Use Spawn Probability=false) {enabled=false}
  ]
  particleSpawnScript [
   &InitializeParticle(
    Color=<rgba 1 1 1 0.5>
    Lifetime=&UniformRangedFloat(Minimum=0.85, Maximum=1.5)
    Mesh Scale=&Multiply_VectorByFloat(
     Vector=&MakeVector(X=0.5, Y=0.5, Z=&RandomRangeFloat(Minimum=0.5, Maximum=0.8))
     Float=&RandomRangeFloat(Minimum=0.8, Maximum=1.1)
    )
    Mesh Scale Mode=::NewEnumerator4
    Sprite Size=&Vector2DFromFloat(Value=50)
    Sprite Size Mode=::NewEnumerator4
    UsePositionOffset=false
    Write Material Random=true
   ) {module=InitializeParticle001}
   /SetVariables(Particles.Burned:NiagaraFloat=&ClampFloat(Float=&FloatFromCurve(FloatCurve=DataInterface() {propertyValues {Curve {keys [
          {arriveTangent=1.1829, interpMode=RCIM_Cubic, leaveTangent=1.1829, tangentMode=RCTM_User, time=0.5, value=0}
          {interpMode=RCIM_Cubic, tangentMode=RCTM_User, time=1, value=1}
         ]}}}))) {isSetParameters=true}
   /SetVariables(Particles.Size:NiagaraFloat=&Multiply_Float(
     A=1
     B=&FloatFromCurve(FloatCurve=DataInterface() {propertyValues {Curve {keys [{interpMode=RCIM_Cubic, tangentMode=RCTM_User, time=0, value=1}, {interpMode=RCIM_Cubic, tangentMode=RCTM_User, time=1, value=0.8}]}}})
    )) {isSetParameters=true}
   /SetVariables(
    Particles.Lifetime=&self:ScratchDynamicInput(Nth:NiagaraInt32=12, ScaleFactor:NiagaraFloat=&RandomRangeFloat(Minimum=1.1, Maximum=1.3))
    Particles.Scale=&self:Scale_Nth_IDx(
     Nth:NiagaraInt32=12
     Scale:Vector3f=$Particles.Scale
     ScaleFactor:NiagaraFloat=&RandomRangeFloat(Minimum=1.1, Maximum=1.3)
    )
   ) {isSetParameters=true}
   /SetVariables(Particles.Magnitude:NiagaraFloat=&RemapRange(Input Value To Remap=$Particles.Burned, Input Min=0, Input Max=1, Output Min=1, Output Max=0.1)) {isSetParameters=true}
   &self:ScratchModule_01(Clip=0.7, Nth=15) {enabled=false, module=ScratchModule_01}
   /SetVariables(Particles.Lifetime=&Add_Float(A=$Particles.Lifetime, B=&Multiply_Float(A=$Particles.Burned, B=0.3))) {isSetParameters=true}
   &SphereLocation(
    Hemisphere Z=true
    MaskSpawnGroup=false
    Non Uniform Scale=<1 1 0.5>
    Offset=&RandomRangeVector(Minimum=<-45 -5 -15>, Maximum=<50 5 -5>)
    Override Seed=false
    Sphere Radius=16
    Surface Only=true
    UseNonUniformScale=true
    UseOffset=true
    UseSphereOrientation=false
   )
  ]
  particleUpdateScript [
   &ParticleState()
   &SolveForcesAndVelocity(Clamp Velocity=false, Limit Acceleration=false)
   &ScaleColor(
    Scale Alpha=&FloatFromCurve(FloatCurve=DataInterface() {propertyValues {Curve {keys [
         {arriveTangent=9.0725, interpMode=RCIM_Cubic, leaveTangent=9.0725, tangentMode=RCTM_User, time=0, value=0}
         {arriveTangent=0.1459, interpMode=RCIM_Cubic, leaveTangent=0.1459, tangentMode=RCTM_User, time=0.1408, value=0.9847}
         {interpMode=RCIM_Cubic, tangentMode=RCTM_User, time=0.4, value=1}
         {arriveTangent=-0.0405, interpMode=RCIM_Cubic, leaveTangent=-0.0405, tangentMode=RCTM_User, time=1, value=0.1}
        ]}}})
    Scale RGB=&Multiply_VectorByFloat(
     Vector=<1 1 1>
     Float=&FloatFromCurve(FloatCurve=DataInterface() {propertyValues {Curve {keys [
          {arriveTangent=9.0725, interpMode=RCIM_Cubic, leaveTangent=9.0725, tangentMode=RCTM_User, time=0, value=0}
          {arriveTangent=-2.5249, interpMode=RCIM_Cubic, leaveTangent=-2.5249, tangentMode=RCTM_User, time=0.1408, value=0.9847}
          {arriveTangent=-0.0405, interpMode=RCIM_Cubic, leaveTangent=-0.0405, tangentMode=RCTM_User, time=1, value=0.1}
         ]}}})
    )
    ScaleA=true
    ScaleRGB=true
   )
   &SpriteFacingAndAlignment(
    Sprite Alignment=&Lerp_Vector(A=$Particles.Velocity, B=<0 0 1>, Alpha=0.5)
    WriteAlignment=true
    WriteFacing=false
   ) {enabled=false}
   &ScaleMeshSize(Scale Factor=&Multiply_VectorByFloat(
     Vector=$Particles.Initial.Scale
     Float=&FloatFromCurve(
      FloatCurve=DataInterface() {propertyValues {Curve {keys [
          <at 0 0.7 RCIM_Cubic>
          {arriveTangent=-0.565, interpMode=RCIM_Cubic, leaveTangent=-0.565, tangentMode=RCTM_User, time=0.2, value=1.5}
          {interpMode=RCIM_Cubic, tangentMode=RCTM_User, time=1, value=0.75}
         ]}}}
      Scale Curve=&Multiply_Float(A=$Particles.Size, B=1)
     )
    )) {module=ScaleMeshSize001}
   /Niagara/Modules/Update/SubUV/V2/SubUVAnimation(
    Start Frame:NiagaraInt32=&RandomRangeInt(Minimum=0, Maximum=10)
    SubUV Loop Count:NiagaraFloat=&RandomRangeFloat(Minimum=1, Maximum=1.25)
   )
   &DynamicMaterialParameters(
    Index 0 Param 1=&Multiply_Float(
     A=&RemapRange(Input Value To Remap=$Particles.Burned, Input Min=0, Input Max=1, Output Min=1, Output Max=0.2)
     B=&RandomRangeFloat(Minimum=1, Maximum=0.2)
    )
    Index 0 Param 2=&RandomRangeFloat(Minimum=0, Maximum=2)
    Index 0 Param 3=8
    Index 0 Param 4=$Particles.SubImageIndex
    Index 1 Param 1=&RandomRangeFloat(Minimum=-1, Maximum=1)
    Index 1 Param 2=&RandomRangeFloat(Minimum=0, Maximum=3)
    Index 1 Param 3=&FloatFromCurve(
     FloatCurve=DataInterface() {propertyValues {Curve {keys [
         {arriveTangent=4.1785, interpMode=RCIM_Cubic, leaveTangent=4.1785, tangentMode=RCTM_User, time=0, value=0}
         {arriveTangent=0.9995, interpMode=RCIM_Cubic, leaveTangent=0.9995, tangentMode=RCTM_User, time=0.2223, value=0.6825}
         {arriveTangent=-0.0405, interpMode=RCIM_Cubic, leaveTangent=-0.0405, tangentMode=RCTM_User, time=0.7, value=0.9}
         <at 1 0.25>
        ]}}}
     Scale Curve=&RemapRange(Input Value To Remap=$Particles.Burned, Input Min=0, Input Max=1, Output Min=3, Output Max=0.5)
    )
    Module.Param0WriteEnabled=true
    Module.Param0WriteEnabled1=true
    Module.Param1WriteEnabled=true
    Module.Param1WriteEnabled1=true
    Module.Param2WriteEnabled=true
    Module.Param2WriteEnabled1=true
    Module.Param3WriteEnabled=true
    Module.Param3WriteEnabled1=false
    Write Parameter Index 1=true
   )
   &CameraOffset(Camera Offset Amount=-2)
  ]
  rendererClasses [&NiagaraMeshRendererProperties]
  renderers [
   {rendererClass=&NiagaraMeshRendererProperties, rendererIndex=0}
   {rendererClass=&NiagaraMeshRendererProperties, rendererIndex=1}
   {rendererClass=&NiagaraMeshRendererProperties, rendererIndex=2}
  ]
 }
 Light_Burst {
  emitterUpdateScript [
   &EmitterState()
   &SpawnBurst_Instantaneous(Use Spawn Probability=false, UseLoopCountLimit=false) {module=SpawnBurst_Instantaneous001}
  ]
  particleSpawnScript [
   &InitializeParticle(
    Color=<rgba 1 0.651 0.3608 1>
    Lifetime=2
    Mesh Scale Mode=::NewEnumerator3
    Mesh Uniform Scale=10
    Sprite Size Mode=::NewEnumerator3
    Uniform Sprite Size=25
    UsePositionOffset=false
    Write Material Random=true
   )
   &SystemLocation(MaskSpawnGroup=false, Offset=<0 0 30>)
   /SetVariables(Particles.LightIntensity=80, Particles.LightRadius=10) {isSetParameters=true}
  ]
  particleUpdateScript [
   &ParticleState()
   /SetVariables(Particles.Position=&Lerp_Vector(
     A=Linked() {linkedVariable {name=Particles.Position, type=NiagaraPosition}}
     B=&Add_Vector(
      A=Linked() {linkedVariable {name=Engine.Owner.Position, type=NiagaraPosition}}
      B=&MakeVector(
       X=&Waveform("Waveform [1]"=::NewEnumerator12, "[1] Frequency"=50, "[1] Amplitude Min/Max"=<-1 1>, "[1] Scale"=100)
       Y=&Waveform("Waveform [1]"=::NewEnumerator12, "[1] Frequency"=25, "[1] Amplitude Min/Max"=<-1 1>, "[1] Scale"=100)
       Z=50
      )
     )
     Alpha=&Multiply_Float(A=$Engine.DeltaTime, B=1)
    )) {isSetParameters=true}
   /SetVariables(Particles.ComponentsEnabled=true) {isSetParameters=true}
   /SetVariables(
    Particles.LightIntensity:NiagaraFloat=&Multiply_Float(
     A=10000
     B=&FloatFromCurve(FloatCurve=DataInterface() {propertyValues {Curve {keys [
          {arriveTangent=32.2034, interpMode=RCIM_Cubic, leaveTangent=32.2034, tangentMode=RCTM_User, time=0, value=0}
          {arriveTangent=-2.5249, interpMode=RCIM_Cubic, leaveTangent=-2.5249, tangentMode=RCTM_User, time=0.1408, value=0.9847}
          {arriveTangent=-0.0405, interpMode=RCIM_Cubic, leaveTangent=-0.0405, tangentMode=RCTM_User, time=1, value=0}
         ]}}})
    )
    Particles.LightRadius=&Multiply_Float(
     A=$Particles.LightRadius
     B=&FloatFromCurve(FloatCurve=DataInterface() {propertyValues {Curve {keys [
          {tangentMode=RCTM_User, time=0, value=1}
          {arriveTangent=-0.0405, interpMode=RCIM_Cubic, leaveTangent=-0.0405, tangentMode=RCTM_User, time=0.5, value=1}
          {arriveTangent=-0.0727, interpMode=RCIM_Cubic, leaveTangent=-0.0727, tangentMode=RCTM_User, time=0.99, value=0}
         ]}}})
    )
   ) {isSetParameters=true}
  ]
  rendererClasses [
   &NiagaraSpriteRendererProperties
   &NiagaraComponentRendererProperties
   &NiagaraLightRendererProperties
  ]
  renderers [
   {rendererClass=&NiagaraSpriteRendererProperties}
   {rendererClass=&NiagaraComponentRendererProperties, rendererIndex=1}
   {rendererClass=&NiagaraLightRendererProperties, rendererIndex=2}
  ]
 }
 Smoke_Burst {
  emitterUpdateScript [
   &EmitterState(
    Enable Distance Culling=true
    MaxDistance=750
    MinDistance=1
    Module.bUseMaxDistance=true
    Module.bUseMinDistance=false
    Scalability Mode=::NewEnumerator1
   )
   &SpawnBurst_Instantaneous(Spawn Count=6, Use Spawn Probability=false, UseLoopCountLimit=false)
  ]
  particleSpawnScript [
   /SetVariables(Particles.RandomFloat=&UniformRangedFloat(Minimum=0, Maximum=1)) {isSetParameters=true}
   &InitializeParticle(
    Color=<rgba 0.099 0.099 0.099 0.8>
    Lifetime=&UniformRangedFloat(Minimum=0.9, Maximum=1.6)
    Mass Mode=::NewEnumerator1
    Sprite Rotation Angle=&UniformRangedFloat(Minimum=0, Maximum=360)
    Sprite Rotation Mode=::NewEnumerator0
    Sprite Size=&Vector2DFromFloat(Value=&Multiply_Float(A=&UniformRangedFloat(Minimum=135, Maximum=180), B=0.5))
    Sprite Size Mode=::NewEnumerator4
    UsePositionOffset=false
    Write Material Random=true
   ) {module=InitializeParticle001}
   &SphereLocation(
    Hemisphere Z=true
    MaskSpawnGroup=false
    Non Uniform Scale=<1 1 0.5>
    Offset=&RandomRangeVector(Minimum=<-40 -5 -15>, Maximum=<50 5 -5>)
    Override Seed=false
    Sphere Radius=16
    Surface Only=true
    UseNonUniformScale=true
    UseOffset=true
    UseSphereOrientation=false
   )
   &AddVelocity(Velocity=<0 0 250>, Velocity Speed Scale=&RandomRangeFloat(Minimum=1, Maximum=1.5))
   &DynamicMaterialParameters(
    Index 0 Param 4=0.25
    Index 2 Param 2=&RandomRangeFloat(Minimum=0, Maximum=1)
    Module.Param0WriteEnabled=false
    Module.Param0WriteEnabled2=false
    Module.Param1WriteEnabled=false
    Module.Param1WriteEnabled2=true
    Module.Param2WriteEnabled=false
    Module.Param2WriteEnabled2=false
    Module.Param3WriteEnabled=true
    Module.Param3WriteEnabled2=false
    Write Parameter Index 2=true
   )
  ]
  particleUpdateScript [
   &ParticleState()
   &GravityForce(Gravity=&UniformRangedVector(Minimum=<0 0 150>, Maximum=<0 0 300>))
   &Drag(Drag=4, Use Linear Drag=true, Use Rotational Drag=false)
   &SolveForcesAndVelocity(Clamp Velocity=false, Limit Acceleration=false)
   &DynamicMaterialParameters(
    Index 0 Param 1=&Add_Float(
     A=$Particles.RandomFloat
     B=&FloatFromCurve(FloatCurve=DataInterface() {propertyValues {Curve {keys [
          {arriveTangent=2.4104, time=0, value=1}
          {arriveTangent=0.2537, interpMode=RCIM_Cubic, leaveTangent=0.2537, tangentMode=RCTM_User, time=1, value=2}
         ]}}})
    )
    Index 0 Param 4=1
    Index 2 Param 1=&FloatFromCurve(FloatCurve=DataInterface() {propertyValues {Curve {keys [{interpMode=RCIM_Cubic, leaveTangent=0.9717, tangentMode=RCTM_Break, time=0.25, value=0.8}, <at 1 1>]}}})
    Index 2 Param 2=&FloatFromCurve(
     FloatCurve=DataInterface() {propertyValues {Curve {keys [<at 0 0.5 RCIM_Cubic>,<at 1 0 RCIM_Cubic>]}}}
     Scale Curve=&RandomRangeFloat(Minimum=0.2, Maximum=0.5)
    )
    Index 2 Param 3=&FloatFromCurve(FloatCurve=DataInterface() {propertyValues {Curve {keys [
         {interpMode=RCIM_Cubic, tangentMode=RCTM_Break, time=0, value=1}
         {interpMode=RCIM_Cubic, leaveTangent=11.1233, tangentMode=RCTM_Break, time=1, value=3}
        ]}}})
    Index 2 Param 4=&FloatFromCurve(
     FloatCurve=DataInterface() {propertyValues {Curve {keys [
         {arriveTangent=-0.4552, interpMode=RCIM_Cubic, leaveTangent=-0.4552, tangentMode=RCTM_User, time=0, value=1}
         {interpMode=RCIM_Cubic, leaveTangent=-1.2956, tangentMode=RCTM_Break, time=1, value=0.75}
        ]}}}
     Scale Curve=1.4
    )
    Module.Param0WriteEnabled=true
    Module.Param0WriteEnabled2=true
    Module.Param1WriteEnabled=false
    Module.Param1WriteEnabled2=true
    Module.Param2WriteEnabled=false
    Module.Param2WriteEnabled2=true
    Module.Param3WriteEnabled=false
    Module.Param3WriteEnabled2=true
    Write Parameter Index 2=true
   ) {module=DynamicMaterialParameters001}
   &ScaleSpriteSize(Scale Factor=&Vector2DFromFloat(Value=&FloatFromCurve(
      FloatCurve=DataInterface() {propertyValues {Curve {keys [
          {arriveTangent=3.358, interpMode=RCIM_Cubic, leaveTangent=3.358, tangentMode=RCTM_User, time=0, value=0.1773}
          {arriveTangent=0.8227, interpMode=RCIM_Cubic, leaveTangent=0.8227, time=0.1179, value=0.489}
          {interpMode=RCIM_Cubic, leaveTangent=0.0321, time=1, value=1}
         ]}}}
      Scale Curve=2
     )))
   &ScaleColor(
    Scale Alpha=&FloatFromCurve(
     FloatCurve=DataInterface() {propertyValues {Curve {keys [
         {arriveTangent=15.3448, interpMode=RCIM_Cubic, leaveTangent=15.3448, tangentMode=RCTM_User, time=0.0001, value=0.0123}
         {arriveTangent=0.1062, interpMode=RCIM_Cubic, leaveTangent=0.1062, tangentMode=RCTM_User, time=0.2, value=1}
         {arriveTangent=-1.5415, interpMode=RCIM_Cubic, leaveTangent=-1.5415, tangentMode=RCTM_User, time=0.7, value=0.75}
         <at 1 0 RCIM_Cubic>
        ]}}}
     Scale Curve=0.75
    )
    ScaleA=true
    ScaleRGB=true
   )
   &SpriteRotationRate(Rotation Rate=&Multiply_Float(
     A=&Lerp_Float(A=-150, B=150, Alpha=$Particles.MaterialRandom)
     B=&Lerp_Float(A=1, B=0, Alpha=&Power(a=$Particles.NormalizedAge, b=0.7))
    ))
   &CameraOffset(Camera Offset Amount=-2)
  ]
  rendererClasses [&NiagaraSpriteRendererProperties]
  renderers [{rendererClass=&NiagaraSpriteRendererProperties}]
 }
}
summary {
 systemName=NS_Emote_BassBurn_GasFire_Burst
 emitters [
  {emitterName=FireLicks_Sprites, rendererClasses [&NiagaraSpriteRendererProperties]}
  {emitterName=Fire_Main, rendererClasses [&NiagaraMeshRendererProperties]}
  {emitterName=Embers_Burst, simTarget=GPUComputeSim, rendererClasses [&NiagaraSpriteRendererProperties]}
  {emitterName=Smoke_Burst, rendererClasses [&NiagaraSpriteRendererProperties]}
  {
   emitterName=Light_Burst
   rendererClasses [
    &NiagaraSpriteRendererProperties
    &NiagaraComponentRendererProperties
    &NiagaraLightRendererProperties
   ]
  }
  {emitterName=FireLicks_Sprites_Tall, rendererClasses [&NiagaraSpriteRendererProperties]}
 ]
 userVariables []
}
