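# Niagara effect for the BassBurn emote that, by its name, shows looping gas splashes, built from mesh and sprite emitters with CPU and GPU variants of each.
# All spawn over a flattened, randomly offset sphere, with rates driven by System.Age curves that fall to zero within about a second.
# Short-lived meshes pop and shrink via vector scale curves.
# Sprite droplets launch in a wide cone, fall under gravity and drag, stretch with speed, and fade.
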NiagaraSystem /SparksCosmetics/Emotes/Bass/BassBurn/Effects/NS_Emote_BassBurn_GasSplash_Loop
dependencies {
 usedDataInterfaces [&NiagaraDataInterfaceCurve, &NiagaraDataInterfaceVectorCurve]
 usedDynamicInputs []
 usedModules [
  &SystemState
  &self:SystemSpawnScript.SystemScriptSource.SystemScriptGraph.NiagaraNodeAssignment_0.SetVariables_#1
 ]
}
emitters {
 GasSplashMesh {
  emitterUpdateScript [
   &EmitterState()
   &SpawnRate(
    SpawnRate=&Multiply_Float(
     A=40
     B=&FloatFromCurve(
      FloatCurve=DataInterface() {propertyValues {Curve {keys [<at 0 1 RCIM_Cubic>, {arriveTangent=-2.2496, interpMode=RCIM_Cubic, leaveTangent=-2.2496, tangentMode=RCTM_User, time=0.9, value=0}]}}}
      CurveIndex=$Engine.System.Age
     )
    )
    Use Spawn Probability=false
   )
  ]
  particleSpawnScript [
   &InitializeParticle(
    Lifetime=0.3
    Mass Mode=::NewEnumerator1
    Mesh Scale=&Multiply_VectorByFloat(Vector=<1 1 1>, Float=&RandomRangeFloat(Minimum=0.2, Maximum=0.6))
    Mesh Scale Mode=::NewEnumerator4
    Position=&SimulationPosition()
    Position Mode=::NewEnumerator1
    Sprite Size Mode=::NewEnumerator4
    UsePositionOffset=false
    Write Material Random=true
   ) {module=InitializeParticle001}
   /SetVariables() {isSetParameters=true, module=SetVariables_#2}
   &InitialMeshOrientation(Rotation=&UniformRangedVector(Minimum=<0 0 0>, Maximum=<0 0 1>), Use Rotation Vector=true) {module=InitialMeshOrientation001}
   &SphereLocation(
    Hemisphere Z=true
    MaskSpawnGroup=false
    Non Uniform Scale=<1 1 0>
    Offset=&RandomRangeVector(Minimum=<-10 -5 -10>, Maximum=<25 5 -10>)
    Override Seed=false
    Sphere Radius=16
    Surface Only=true
    UseNonUniformScale=true
    UseOffset=true
    UseSphereOrientation=false
   ) {module=SphereLocation001}
   &ShapeLocation(MaskSpawnGroup=false, Override Seed=false, Ring Radius=1, Shape Primitive=::NewEnumerator4, UseOwnerScale=true) {enabled=false}
   &AddVelocity(Velocity=<0 0 5>) {enabled=false, module=AddVelocity001}
  ]
  particleUpdateScript [
   &ParticleState()
   /SetVariables(Particles.Position=&Lerp_Vector(
     A=Linked() {linkedVariable {name=Particles.Position, type=NiagaraPosition}}
     B=&Add_Vector(
      A=Linked() {linkedVariable {name=Engine.Owner.Position, type=NiagaraPosition}}
      B=&MakeVector(
       X=&Waveform("Waveform [1]"=::NewEnumerator12, "[1] Frequency"=50, "[1] Amplitude Min/Max"=<-1 1>, "[1] Scale"=100)
       Y=&Waveform("Waveform [1]"=::NewEnumerator12, "[1] Frequency"=25, "[1] Amplitude Min/Max"=<-1 1>, "[1] Scale"=100)
       Z=50
      )
     )
     Alpha=&Multiply_Float(A=$Engine.DeltaTime, B=1)
    )) {enabled=false, isSetParameters=true, module=SetVariables_#3}
   &ScaleColor(
    Scale Alpha=&FloatFromCurve(FloatCurve=DataInterface() {propertyValues {Curve {keys [<at 0.65 1.0113>,<at 1.0011 0.002>]}}})
    ScaleA=true
    ScaleRGB=true
   )
   &Drag(Use Linear Drag=true, Use Rotational Drag=false)
   &GravityForce(Gravity=<0 0 -5>)
   &SolveForcesAndVelocity(Clamp Velocity=false, Limit Acceleration=false)
   &ScaleMeshSize(Scale Factor=&VectorFromCurve(VectorCurve=DataInterface() {propertyValues {
       XCurve {keys [{interpMode=RCIM_Cubic, leaveTangent=4.8413, tangentMode=RCTM_Break, time=0, value=0}, <at 0.7127 0.5793>, <at 1 0>]}
       YCurve {keys [{interpMode=RCIM_Cubic, leaveTangent=4.8413, tangentMode=RCTM_Break, time=0, value=0}, <at 0.7127 0.5793>, <at 1 0>]}
       ZCurve {keys [
         {interpMode=RCIM_Cubic, leaveTangent=11.3233, tangentMode=RCTM_Break, time=0, value=0}
         {arriveTangent=-3.7658, interpMode=RCIM_Cubic, leaveTangent=-3.7658, tangentMode=RCTM_User, time=0.0765, value=1.4}
         <at 0.9989 0.1>
        ]}
      }}))
  ]
  rendererClasses [&NiagaraMeshRendererProperties]
  renderers [{rendererClass=&NiagaraMeshRendererProperties, rendererIndex=0}]
 }
 GasSplashMeshGPU {
  simTarget=GPUComputeSim
  emitterUpdateScript [
   &EmitterState()
   &SpawnRate(
    SpawnRate=&Multiply_Float(
     A=60
     B=&FloatFromCurve(
      FloatCurve=DataInterface() {propertyValues {Curve {keys [<at 0 1 RCIM_Cubic>, {arriveTangent=-2.2496, interpMode=RCIM_Cubic, leaveTangent=-2.2496, tangentMode=RCTM_User, time=0.9, value=0}]}}}
      CurveIndex=$Engine.System.Age
     )
    )
    Use Spawn Probability=false
   )
  ]
  particleSpawnScript [
   &InitializeParticle(
    Lifetime=0.3
    Mass Mode=::NewEnumerator1
    Mesh Scale=&Multiply_VectorByFloat(Vector=<1 1 1>, Float=&RandomRangeFloat(Minimum=0.2, Maximum=0.6))
    Mesh Scale Mode=::NewEnumerator4
    Position=&SimulationPosition()
    Position Mode=::NewEnumerator1
    Sprite Size Mode=::NewEnumerator4
    UsePositionOffset=false
    Write Material Random=true
   ) {module=InitializeParticle001}
   /SetVariables() {isSetParameters=true, module=SetVariables_#2}
   &InitialMeshOrientation(Rotation=&UniformRangedVector(Minimum=<0 0 0>, Maximum=<0 0 1>), Use Rotation Vector=true) {module=InitialMeshOrientation001}
   &SphereLocation(
    Hemisphere Z=true
    MaskSpawnGroup=false
    Non Uniform Scale=<1 1 0>
    Offset=&RandomRangeVector(Minimum=<-10 -5 -10>, Maximum=<25 5 -10>)
    Override Seed=false
    Sphere Radius=16
    Surface Only=true
    UseNonUniformScale=true
    UseOffset=true
    UseSphereOrientation=false
   ) {module=SphereLocation001}
   &ShapeLocation(MaskSpawnGroup=false, Override Seed=false, Ring Radius=1, Shape Primitive=::NewEnumerator4, UseOwnerScale=true) {enabled=false}
   &AddVelocity(Velocity=<0 0 5>) {enabled=false, module=AddVelocity001}
  ]
  particleUpdateScript [
   &ParticleState()
   /SetVariables(Particles.Position=&Lerp_Vector(
     A=Linked() {linkedVariable {name=Particles.Position, type=NiagaraPosition}}
     B=&Add_Vector(
      A=Linked() {linkedVariable {name=Engine.Owner.Position, type=NiagaraPosition}}
      B=&MakeVector(
       X=&Waveform("Waveform [1]"=::NewEnumerator12, "[1] Frequency"=50, "[1] Amplitude Min/Max"=<-1 1>, "[1] Scale"=100)
       Y=&Waveform("Waveform [1]"=::NewEnumerator12, "[1] Frequency"=25, "[1] Amplitude Min/Max"=<-1 1>, "[1] Scale"=100)
       Z=50
      )
     )
     Alpha=&Multiply_Float(A=$Engine.DeltaTime, B=1)
    )) {enabled=false, isSetParameters=true, module=SetVariables_#3}
   &ScaleColor(
    Scale Alpha=&FloatFromCurve(FloatCurve=DataInterface() {propertyValues {Curve {keys [<at 0.65 1.0113>,<at 1.0011 0.002>]}}})
    ScaleA=true
    ScaleRGB=true
   )
   &Drag(Use Linear Drag=true, Use Rotational Drag=false)
   &GravityForce(Gravity=<0 0 -5>)
   &SolveForcesAndVelocity(Clamp Velocity=false, Limit Acceleration=false)
   &ScaleMeshSize(Scale Factor=&VectorFromCurve(VectorCurve=DataInterface() {propertyValues {
       XCurve {keys [{interpMode=RCIM_Cubic, leaveTangent=4.8413, tangentMode=RCTM_Break, time=0, value=0}, <at 0.7127 0.5793>, <at 1 0>]}
       YCurve {keys [{interpMode=RCIM_Cubic, leaveTangent=4.8413, tangentMode=RCTM_Break, time=0, value=0}, <at 0.7127 0.5793>, <at 1 0>]}
       ZCurve {keys [
         {interpMode=RCIM_Cubic, leaveTangent=11.3233, tangentMode=RCTM_Break, time=0, value=0}
         {arriveTangent=-3.7658, interpMode=RCIM_Cubic, leaveTangent=-3.7658, tangentMode=RCTM_User, time=0.0765, value=1.4}
         <at 0.9989 0.1>
        ]}
      }}))
  ]
  rendererClasses [&NiagaraMeshRendererProperties]
  renderers [{rendererClass=&NiagaraMeshRendererProperties, rendererIndex=0}]
 }
 GasSplashSprite {
  emitterUpdateScript [
   &EmitterState(Life Cycle Mode=::NewEnumerator1, Loop Behavior=::NewEnumerator1, Loop Duration=1, UseLoopDelay=false)
   &SpawnRate(
    SpawnRate=&Multiply_Float(
     A=40
     B=&FloatFromCurve(
      FloatCurve=DataInterface() {propertyValues {Curve {keys [
          <at 0 1>
          {arriveTangent=-1, interpMode=RCIM_Cubic, leaveTangent=-1, time=0.9, value=0.9}
          {arriveTangent=-2.343, interpMode=RCIM_Cubic, leaveTangent=-2.343, tangentMode=RCTM_User, time=1, value=0}
         ]}}}
      CurveIndex=$Engine.System.Age
     )
    )
    Use Spawn Probability=false
   ) {module=SpawnRate001}
  ]
  particleSpawnScript [
   &InitializeParticle(
    Color=<rgba 1 1 1 3>
    Lifetime=&UniformRangedFloat(Minimum=0.2, Maximum=0.3)
    Mass Mode=::NewEnumerator1
    Mesh Scale=<0.5 0.5 0.5>
    Mesh Scale Mode=::NewEnumerator4
    Sprite Size=&Multiply_Vector2(
     A=<4 6>
     B=&Vector2DFromFloat(Value=&UniformRangedFloat(Minimum=0.1, Maximum=1))
    )
    Sprite Size Mode=::NewEnumerator4
    UsePositionOffset=false
    Write Material Random=true
   ) {module=InitializeParticle001}
   /SetVariables(Particles.Scale=/Game/Effects/Niagara/DynamicInputs/Water/FN_PickVectorFromSpawnGroupCutoff(
     Under Cutoff Vector:Vector3f=<0 0 0>
     Over Cutoff Vector:Vector3f=<1 1 1>
     SpawnGroup:NiagaraInt32=$Emitter.SpawnGroup
     Cutoff:NiagaraInt32=2
    )) {isSetParameters=true, module=SetVariables_#4}
   &SphereLocation(
    Hemisphere Z=true
    MaskSpawnGroup=false
    Non Uniform Scale=<1 1 0>
    Offset=&RandomRangeVector(Minimum=<-10 -5 -15>, Maximum=<25 5 -5>)
    Override Seed=false
    Sphere Radius=16
    Surface Only=true
    UseNonUniformScale=true
    UseOffset=true
    UseSphereOrientation=false
   ) {module=SphereLocation001}
   &CylinderLocation(
    Cylinder Height=0
    Cylinder Height Midpoint=0.1
    Cylinder Radius=2
    MaskSpawnGroup=false
    Offset=<0 0 1>
    Offset Coordinate Space:ENiagaraCoordinateSpace=::World
    Offset Mode:ENiagara_OffsetMode=::NewEnumerator0
    Override Seed=false
    UseOwnerScale=true
   ) {enabled=false}
   &AddVelocity(Velocity=<0 0 5>) {enabled=false, module=AddVelocity001}
   &AddVelocity(
    Cone Angle=180
    Cone Axis=<0 0 1>
    Module.UseConeFalloff=false
    Module.UseVelDistribution=false
    Override Seed=false
    Velocity Mode=::NewEnumerator2
    Velocity Speed=&RandomRangeFloat(Minimum=175, Maximum=225)
   ) {module=AddVelocity002}
  ]
  particleUpdateScript [
   &ParticleState()
   /SetVariables(Particles.Position=&Lerp_Vector(
     A=Linked() {linkedVariable {name=Particles.Position, type=NiagaraPosition}}
     B=&Add_Vector(
      A=Linked() {linkedVariable {name=Engine.Owner.Position, type=NiagaraPosition}}
      B=&MakeVector(
       X=&Waveform("Waveform [1]"=::NewEnumerator12, "[1] Frequency"=50, "[1] Amplitude Min/Max"=<-1 1>, "[1] Scale"=100)
       Y=&Waveform("Waveform [1]"=::NewEnumerator12, "[1] Frequency"=25, "[1] Amplitude Min/Max"=<-1 1>, "[1] Scale"=100)
       Z=50
      )
     )
     Alpha=&Multiply_Float(A=$Engine.DeltaTime, B=1)
    )) {enabled=false, isSetParameters=true, module=SetVariables_#5}
   &GravityForce(Gravity=<0 0 -500>)
   &ScaleColor(
    Scale Alpha=&FloatFromCurve(FloatCurve=DataInterface() {propertyValues {Curve {keys [<at 0.4124 1.0113>,<at 1.0011 0.002>]}}})
    ScaleA=true
    ScaleRGB=false
   )
   &Drag(Drag=5, Use Linear Drag=true, Use Rotational Drag=false)
   &SolveForcesAndVelocity(Clamp Velocity=false, Limit Acceleration=false)
   &ScaleSpriteSizeBySpeed(Max Scale Factor=<1 5>, Velocity Threshold=500)
   &ScaleSpriteSize(Scale Factor=&Vector2DFromFloat(Value=/Game/Effects/Niagara/DynamicInputs/FN_PickFloatFromSpawnGroupCutoff(
      Over Cutoff Float:NiagaraFloat=0
      Under Cutoff Float:NiagaraFloat=&FloatFromCurve(FloatCurve=DataInterface() {propertyValues {Curve {keys [
           {interpMode=RCIM_Cubic, tangentMode=RCTM_Break, time=0.5, value=0.9489}
           {arriveTangent=-3.7049, interpMode=RCIM_Cubic, tangentMode=RCTM_Break, time=1.0011, value=0.002}
          ]}}})
      SpawnGroup:NiagaraInt32=$Emitter.SpawnGroup
      Cutoff:NiagaraInt32=2
     )))
  ]
  rendererClasses [&NiagaraSpriteRendererProperties]
  renderers [{rendererClass=&NiagaraSpriteRendererProperties}]
 }
 GasSplashSpriteGPU {
  simTarget=GPUComputeSim
  emitterUpdateScript [
   &EmitterState(Life Cycle Mode=::NewEnumerator1, Loop Behavior=::NewEnumerator1, Loop Duration=1, UseLoopDelay=false)
   &SpawnRate(
    SpawnRate=&Multiply_Float(
     A=150
     B=&FloatFromCurve(
      FloatCurve=DataInterface() {propertyValues {Curve {keys [
          <at 0 1>
          {arriveTangent=-1, interpMode=RCIM_Cubic, leaveTangent=-1, time=0.9, value=0.9}
          {arriveTangent=-2.343, interpMode=RCIM_Cubic, leaveTangent=-2.343, tangentMode=RCTM_User, time=1, value=0}
         ]}}}
      CurveIndex=$Engine.System.Age
     )
    )
    Use Spawn Probability=false
   ) {module=SpawnRate001}
  ]
  particleSpawnScript [
   &InitializeParticle(
    Color=<rgba 1 1 1 3>
    Lifetime=&UniformRangedFloat(Minimum=0.2, Maximum=0.3)
    Mass Mode=::NewEnumerator1
    Mesh Scale=<0.5 0.5 0.5>
    Mesh Scale Mode=::NewEnumerator4
    Sprite Size=&Multiply_Vector2(
     A=<4 6>
     B=&Vector2DFromFloat(Value=&UniformRangedFloat(Minimum=0.1, Maximum=1))
    )
    Sprite Size Mode=::NewEnumerator4
    UsePositionOffset=false
    Write Material Random=true
   ) {module=InitializeParticle001}
   /SetVariables(Particles.Scale=&FN_PickVectorFromSpawnGroupCutoff(
     Under Cutoff Vector:Vector3f=<0 0 0>
     Over Cutoff Vector:Vector3f=<1 1 1>
     SpawnGroup:NiagaraInt32=$Emitter.SpawnGroup
     Cutoff:NiagaraInt32=2
    )) {isSetParameters=true, module=SetVariables_#4}
   &SphereLocation(
    Hemisphere Z=true
    MaskSpawnGroup=false
    Non Uniform Scale=<1 1 0>
    Offset=&RandomRangeVector(Minimum=<-10 -5 -15>, Maximum=<25 5 -5>)
    Override Seed=false
    Sphere Radius=16
    Surface Only=true
    UseNonUniformScale=true
    UseOffset=true
    UseSphereOrientation=false
   ) {module=SphereLocation001}
   &CylinderLocation(
    Cylinder Height=0
    Cylinder Height Midpoint=0.1
    Cylinder Radius=2
    MaskSpawnGroup=false
    Offset=<0 0 1>
    Offset Coordinate Space:ENiagaraCoordinateSpace=::World
    Offset Mode:ENiagara_OffsetMode=::NewEnumerator0
    Override Seed=false
    UseOwnerScale=true
   ) {enabled=false}
   &AddVelocity(Velocity=<0 0 5>) {enabled=false, module=AddVelocity001}
   &AddVelocity(
    Cone Angle=180
    Cone Axis=<0 0 1>
    Module.UseConeFalloff=false
    Module.UseVelDistribution=false
    Override Seed=false
    Velocity Mode=::NewEnumerator2
    Velocity Speed=&RandomRangeFloat(Minimum=175, Maximum=225)
   ) {module=AddVelocity002}
  ]
  particleUpdateScript [
   &ParticleState()
   /SetVariables(Particles.Position=&Lerp_Vector(
     A=Linked() {linkedVariable {name=Particles.Position, type=NiagaraPosition}}
     B=&Add_Vector(
      A=Linked() {linkedVariable {name=Engine.Owner.Position, type=NiagaraPosition}}
      B=&MakeVector(
       X=&Waveform("Waveform [1]"=::NewEnumerator12, "[1] Frequency"=50, "[1] Amplitude Min/Max"=<-1 1>, "[1] Scale"=100)
       Y=&Waveform("Waveform [1]"=::NewEnumerator12, "[1] Frequency"=25, "[1] Amplitude Min/Max"=<-1 1>, "[1] Scale"=100)
       Z=50
      )
     )
     Alpha=&Multiply_Float(A=$Engine.DeltaTime, B=1)
    )) {enabled=false, isSetParameters=true, module=SetVariables_#5}
   &GravityForce(Gravity=<0 0 -500>)
   &ScaleColor(
    Scale Alpha=&FloatFromCurve(FloatCurve=DataInterface() {propertyValues {Curve {keys [<at 0.4124 1.0113>,<at 1.0011 0.002>]}}})
    ScaleA=true
    ScaleRGB=false
   )
   &Drag(Drag=5, Use Linear Drag=true, Use Rotational Drag=false)
   &SolveForcesAndVelocity(Clamp Velocity=false, Limit Acceleration=false)
   &ScaleSpriteSizeBySpeed(Max Scale Factor=<1 5>, Velocity Threshold=500)
   &ScaleSpriteSize(Scale Factor=&Vector2DFromFloat(Value=&FN_PickFloatFromSpawnGroupCutoff(
      Over Cutoff Float:NiagaraFloat=0
      Under Cutoff Float:NiagaraFloat=&FloatFromCurve(FloatCurve=DataInterface() {propertyValues {Curve {keys [
           {interpMode=RCIM_Cubic, tangentMode=RCTM_Break, time=0.5, value=0.9489}
           {arriveTangent=-3.7049, interpMode=RCIM_Cubic, tangentMode=RCTM_Break, time=1.0011, value=0.002}
          ]}}})
      SpawnGroup:NiagaraInt32=$Emitter.SpawnGroup
      Cutoff:NiagaraInt32=2
     )))
  ]
  rendererClasses [&NiagaraSpriteRendererProperties]
  renderers [{rendererClass=&NiagaraSpriteRendererProperties}]
 }
}
summary {
 systemName=NS_Emote_BassBurn_GasSplash_Loop
 emitters [
  {emitterName=GasSplashSprite, rendererClasses [&NiagaraSpriteRendererProperties]}
  {emitterName=GasSplashMesh, rendererClasses [&NiagaraMeshRendererProperties]}
  {emitterName=GasSplashSpriteGPU, simTarget=GPUComputeSim, rendererClasses [&NiagaraSpriteRendererProperties]}
  {emitterName=GasSplashMeshGPU, simTarget=GPUComputeSim, rendererClasses [&NiagaraMeshRendererProperties]}
 ]
 userVariables []
}
