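# Niagara effect for gasoline pouring from a gas can in the BassBurn bass emote.
# A gravity-driven ribbon forms the stream, layered with BakeFluid sprites animated through dynamic material parameters.
# Near-identical CPU and GPU drip emitters add velocity-stretched droplets killed by a position threshold, with curves over age tapering the pour.
# A GPU PIC/FLIP 2D-grid fluid simulation with camera-facing grid and render-target surfacing is also present but disabled.
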NiagaraSystem /SparksCosmetics/Emotes/Bass/BassBurn/Effects/NS_GasCan_Pour_Fluid
dependencies {
 usedDataInterfaces [
  &NiagaraDataInterfaceCurve
  &NiagaraDataInterfaceVector2DCurve
  /Script/Niagara.NiagaraDataInterfaceGrid2DCollection
  /Script/Niagara.NiagaraDataInterfaceRenderTarget2D
  &NiagaraDataInterfaceCamera
  /Script/Niagara.NiagaraDataInterfaceRasterizationGrid3D
  &NiagaraDataInterfacePlatformSet
  /Script/Niagara.NiagaraDataInterfaceEmitterProperties
 ]
 usedDynamicInputs [&Lerp_Float]
 usedModules [
  &SystemState
  &self:SystemSpawnScript.SystemScriptSource.SystemScriptGraph.NiagaraNodeAssignment_0.SetVariables_#1
  &self:SystemSpawnScript.SystemScriptSource.SystemScriptGraph.NiagaraNodeAssignment_1.SetVariables_#2
 ]
}
emitters {
 BakeFluid {
  emitterUpdateScript [
   &EmitterState()
   &SpawnRate(
    SpawnRate=&FloatFromCurve(
     FloatCurve=DataInterface() {propertyValues {Curve {keys [
         {arriveTangent=3.4706, interpMode=RCIM_Cubic, leaveTangent=3.4706, tangentMode=RCTM_User, time=0, value=0}
         {arriveTangent=0.3451, interpMode=RCIM_Cubic, leaveTangent=0.3451, tangentMode=RCTM_User, time=0.1, value=1}
         {arriveTangent=-0.003, interpMode=RCIM_Cubic, leaveTangent=-0.003, tangentMode=RCTM_User, time=0.9, value=0}
        ]}}}
     CurveIndex=$System.Age
     Scale Curve=20
    )
    Use Spawn Probability=false
   )
  ]
  particleSpawnScript [
   &InitializeParticle(
    Color Mode=::NewEnumerator0
    Lifetime Max=0.5
    Lifetime Min=0.3
    Lifetime Mode=::NewEnumerator1
    Sprite Size Mode=::NewEnumerator1
    Sprite UV Mode=::NewEnumerator3
    Uniform Sprite Size Max=50
    Uniform Sprite Size Min=40
    UsePositionOffset=false
    Write Material Random=true
   )
   &AddVelocity(Velocity=<30 0 50>) {enabled=false}
   &CameraOffset(Camera Offset Amount=60)
  ]
  particleUpdateScript [
   &ParticleState()
   &ScaleSpriteSize(Scale Factor=<0.6 2>, Scale Sprite Size Mode=::NewEnumerator2) {module=ScaleSpriteSize001}
   &ScaleSpriteSize(Uniform Curve Sprite Scale=DataInterface() {propertyValues {Curve {keys [
        {arriveTangent=0.0076, interpMode=RCIM_Cubic, leaveTangent=0.0076, tangentMode=RCTM_User, time=0, value=0.3}
        {interpMode=RCIM_Cubic, tangentMode=RCTM_User, time=1, value=0.6}
       ]}}})
   &DynamicMaterialParameters(
    Index 0 Param 1=&FloatFromCurve(FloatCurve=DataInterface() {propertyValues {Curve {keys [<at 0 0.01 RCIM_Cubic>, {arriveTangent=1.1519, interpMode=RCIM_Cubic, leaveTangent=1.1519, tangentMode=RCTM_User, time=1, value=1}]}}})
    Index 0 Param 2=&FloatFromCurve(FloatCurve=DataInterface() {propertyValues {Curve {keys [{arriveTangent=0.4361, time=0.2, value=0.05}, <at 1 0.7>]}}})
    Index 0 Param 3=&FloatFromCurve(FloatCurve=DataInterface() {propertyValues {Curve {keys [
         {arriveTangent=0.4361, interpMode=RCIM_Cubic, tangentMode=RCTM_Break, time=0.5, value=0}
         {arriveTangent=-1.6488, interpMode=RCIM_Cubic, tangentMode=RCTM_Break, time=1, value=-0.5}
        ]}}})
    Module.Param0WriteEnabled=true
    Module.Param1WriteEnabled=true
    Module.Param2WriteEnabled=true
    Module.Param3WriteEnabled=true
   )
   &GravityForce(Gravity=<0 0 -600>)
   &SolveForcesAndVelocity(Clamp Velocity=false, Limit Acceleration=false)
   &ScaleSpriteSizeBySpeed(Max Scale Factor=<3 3>)
  ]
  rendererClasses [&NiagaraSpriteRendererProperties]
  renderers [{rendererClass=&NiagaraSpriteRendererProperties}]
 }
 Drips_CPU_Low {
  emitterUpdateScript [
   &EmitterState()
   &SpawnRate(
    SpawnRate=&Multiply_Float(
     A=50
     B=&FloatFromCurve(FloatCurve=DataInterface() {propertyValues {Curve {keys [<at 0 1 RCIM_Cubic>,<at 1.1 0>]}}}, CurveIndex=$Engine.System.Age)
    )
    Use Spawn Probability=false
   )
  ]
  particleSpawnScript [
   &InitializeParticle(
    Lifetime=&RandomRangeFloat(Minimum=0.5, Maximum=0.8)
    Mass Mode=::NewEnumerator1
    Sprite Size=&RandomRangeVector2D(Minimum=<0.85 1>, Maximum=<2 3>)
    Sprite Size Mode=::NewEnumerator4
    UsePositionOffset=false
    Write Material Random=true
   ) {module=InitializeParticle001}
   &AddVelocity(
    Scale Added Velocity=&VectorFromFloat(Value=$System.VelocityMultiplier)
    Velocity=&RandomRangeVector(Minimum=<25 -10 -10>, Maximum=<50 10 10>, Evaluation Type=::NewEnumerator1)
   ) {enabled=false}
   /Niagara/Modules/Spawn/Location/SphereLocation(MaskSpawnGroup=false, Module.Override Randomness=false, Sphere Radius=1, Surface Only=false)
   &AddVelocity(Velocity=&Multiply_VectorByFloat(
     Vector=$Output.SphereLocation.SphereVector
     Float=&RandomRangeFloat(Minimum=10, Maximum=20, Evaluation Type=::NewEnumerator1)
    )) {module=AddVelocity001}
  ]
  particleUpdateScript [
   &ParticleState()
   &GravityForce()
   &Drag(Use Linear Drag=true, Use Rotational Drag=false)
   &SolveForcesAndVelocity(Clamp Velocity=false, Limit Acceleration=false)
   &Color(Color=&MakeLinearColorFromVectorAndFloat(
     "Vector (RGB)"=&VectorFromFloat(Value=3)
     "Float (Alpha)"=&ClampFloat(Float=&FloatFromCurve(FloatCurve=DataInterface() {propertyValues {Curve {keys [{arriveTangent=-1, time=0, value=1}, <at 1 0>]}}}))
    ))
   &ScaleSpriteSizeBySpeed(Max Scale Factor=<1 25>, Source Velocity:Vector3f=&Multiply_Vector(A=<0 0.05 1>, B=$Particles.Velocity))
   &ScaleSpriteSize(Scale Factor=&Vector2DFromCurve(Vector2Curve=DataInterface() {propertyValues {XCurve {keys [<at 0.9 1>,<at 1 0>]}, YCurve {keys [<at 0.9 1>,<at 1 0>]}}}, ScaleCurve=<1 1>))
   &KillParticles(Kill Particles=&SetBoolByFloatComparison(
     A=&MakeFloatFromVector(Vector=Linked() {linkedVariable {name=Particles.Position, type=NiagaraPosition}}, Channel=::NewEnumerator2)
     B=0
     Comparison Type=::NewEnumerator6
    ))
  ]
  rendererClasses [&NiagaraSpriteRendererProperties]
  renderers [{rendererClass=&NiagaraSpriteRendererProperties}]
 }
 Drips_GPU {
  simTarget=GPUComputeSim
  emitterUpdateScript [
   &EmitterState(Inactive Response=::NewEnumerator1, Life Cycle Mode=::NewEnumerator1, Loop Behavior=::NewEnumerator1, Loop Duration=4, UseLoopDelay=true)
   &SpawnRate(
    SpawnRate=&Multiply_Float(
     A=100
     B=&FloatFromCurve(
      FloatCurve=DataInterface() {propertyValues {Curve {keys [
          {arriveTangent=3.4706, interpMode=RCIM_Cubic, leaveTangent=3.4706, tangentMode=RCTM_User, time=0, value=0}
          {arriveTangent=0.3451, interpMode=RCIM_Cubic, leaveTangent=0.3451, tangentMode=RCTM_User, time=0.1, value=1}
          {arriveTangent=-0.0954, interpMode=RCIM_Cubic, leaveTangent=-0.0954, tangentMode=RCTM_User, time=1.1, value=0}
         ]}}}
      CurveIndex=$Engine.System.Age
     )
    )
    Use Spawn Probability=false
   )
  ]
  particleSpawnScript [
   &InitializeParticle(
    Lifetime=&RandomRangeFloat(Minimum=0.5, Maximum=0.8)
    Mass Mode=::NewEnumerator1
    Sprite Size=&RandomRangeVector2D(Minimum=<1 1>, Maximum=<3 3>)
    Sprite Size Mode=::NewEnumerator4
    UsePositionOffset=false
    Write Material Random=true
   ) {module=InitializeParticle001}
   &AddVelocity(
    Scale Added Velocity=&VectorFromFloat(Value=$System.VelocityMultiplier)
    Velocity=&RandomRangeVector(Minimum=<25 -10 -10>, Maximum=<50 10 10>, Evaluation Type=::NewEnumerator1)
   ) {enabled=false}
   &SphereLocation(MaskSpawnGroup=false, Module.Override Randomness=false, Sphere Radius=1, Surface Only=false)
   &AddVelocity(Velocity=&Multiply_VectorByFloat(
     Vector=$Output.SphereLocation.SphereVector
     Float=&RandomRangeFloat(Minimum=10, Maximum=20, Evaluation Type=::NewEnumerator1)
    )) {module=AddVelocity001}
  ]
  particleUpdateScript [
   &ParticleState()
   &GravityForce()
   &Drag(Use Linear Drag=true, Use Rotational Drag=false)
   &SolveForcesAndVelocity(Clamp Velocity=false, Limit Acceleration=false)
   &Color(Color=&MakeLinearColorFromVectorAndFloat(
     "Vector (RGB)"=&VectorFromFloat(Value=3)
     "Float (Alpha)"=&ClampFloat(Float=&FloatFromCurve(FloatCurve=DataInterface() {propertyValues {Curve {keys [{arriveTangent=-1, time=0, value=1}, <at 1 0>]}}}))
    ))
   &ScaleSpriteSizeBySpeed(Max Scale Factor=<1 25>, Source Velocity:Vector3f=&Multiply_Vector(A=<0 0.05 1>, B=$Particles.Velocity))
   &ScaleSpriteSize(Scale Factor=&Vector2DFromCurve(Vector2Curve=DataInterface() {propertyValues {XCurve {keys [<at 0.9 1>,<at 1 0>]}, YCurve {keys [<at 0.9 1>,<at 1 0>]}}}, ScaleCurve=<1 1>))
   &KillParticles(Kill Particles=&SetBoolByFloatComparison(
     A=&MakeFloatFromVector(Vector=Linked() {linkedVariable {name=Particles.Position, type=NiagaraPosition}}, Channel=::NewEnumerator2)
     B=0
     Comparison Type=::NewEnumerator6
    ))
  ]
  rendererClasses [&NiagaraSpriteRendererProperties]
  renderers [{rendererClass=&NiagaraSpriteRendererProperties}]
 }
 FN_PIC_FLIP_Fluid_Stream_Gas {
  bEnabled=false
  simTarget=GPUComputeSim
  emitterSpawnScript [
   /SetVariables(
    Emitter.PressureGrid=DataInterface() {propertyValues {
      ClearBeforeNonIterationStage=true
      NumAttributes=0
      NumCellsMaxAxis=3
      NumCellsX=3
      NumCellsY=3
      OverrideBufferFormat=Float
      PreviewAttribute=None
      SetGridFromMaxAxis=false
      bOverrideFormat=true
      bPreviewGrid=false
      RenderTargetUserParameter {parameter {name=None, typeDefHandle {registeredTypeIndex=4}, varData []}}
      WorldBBoxSize=<100 100>
     }}
    Emitter.VelocityGrid=DataInterface() {propertyValues {
      ClearBeforeNonIterationStage=true
      NumAttributes=0
      NumCellsMaxAxis=3
      NumCellsX=128
      NumCellsY=64
      OverrideBufferFormat=Float
      PreviewAttribute=None
      SetGridFromMaxAxis=false
      bOverrideFormat=false
      bPreviewGrid=false
      RenderTargetUserParameter {parameter {name=None, typeDefHandle {registeredTypeIndex=4}, varData []}}
      WorldBBoxSize=<100 100>
     }}
   ) {isSetParameters=true}
   /NiagaraFluids/Modules/Grid2D/Grid2D_SetResolution(
    Grid:NiagaraDataInterfaceGrid2DCollection=$Emitter.VelocityGrid
    NumCellsMaxAxis=256
    Resolution Method:ENiagaraGrid2DResolution=::NewEnumerator1
    WorldGridExtents:Vector2f=<256 256>
   )
   &Grid2D_SetResolution(
    Grid:NiagaraDataInterfaceGrid2DCollection=$Emitter.PressureGrid
    OtherGrid:NiagaraDataInterfaceGrid2DCollection=$Emitter.VelocityGrid
    Resolution Method:ENiagaraGrid2DResolution=::NewEnumerator3
    WorldGridExtents:Vector2f=$Emitter.Grid2D_SetResolution.WorldGridExtents
   ) {module=Grid2D_SetResolution002}
   /SetVariables(Emitter.DeltaTime=0.033) {isSetParameters=true}
   /SetVariables(Emitter.SimRT=DataInterface() {propertyValues {
      MipMapGeneration=Disabled
      MipMapGenerationType=Linear
      OverrideRenderTargetFilter=TF_Default
      OverrideRenderTargetFormat=RTF_RGBA16f
      bInheritUserParameterSettings=false
      bOverrideFormat=true
      bPreviewRenderTarget=false
      RenderTargetUserParameter {parameter {name=None, typeDefHandle {registeredTypeIndex=4}, varData []}}
      Size=<0 0>
     }}) {isSetParameters=true}
   /NiagaraFluids/Modules/Grid2D/Grid2D_InitRT(Grid:NiagaraDataInterfaceGrid2DCollection=$Emitter.VelocityGrid, SimRT:NiagaraDataInterfaceRenderTarget2D=$Emitter.SimRT) {module=InitRTGrid2D}
   &self:FN_PIC_FLIP_Fluid_Stream_Gas_1.NiagaraScratchPadContainer_0.ScratchModule_0() {module=ScratchModule}
   &self:FN_PIC_FLIP_Fluid_Stream_Gas_1.NiagaraScratchPadContainer_0.ScratchModule_03_5(NumPressureIterations=10) {module=ScratchModule_03}
   /SetVariables() {isSetParameters=true}
   /SetVariables(Emitter.ThicknessRT=DataInterface() {propertyValues {
      MipMapGeneration=Disabled
      MipMapGenerationType=Linear
      OverrideRenderTargetFilter=TF_Default
      OverrideRenderTargetFormat=RTF_R16f
      bInheritUserParameterSettings=false
      bOverrideFormat=true
      bPreviewRenderTarget=false
      RenderTargetUserParameter {parameter {name=None, typeDefHandle {registeredTypeIndex=4}, varData []}}
      Size=<0 0>
     }}) {isSetParameters=true}
   &Grid2D_InitRT(Grid:NiagaraDataInterfaceGrid2DCollection=$Emitter.VelocityGrid, SimRT:NiagaraDataInterfaceRenderTarget2D=$Emitter.ThicknessRT)
   /SetVariables(
    Emitter.RasterizationGrid=DataInterface() {propertyValues {
      CellSize=1
      ClearBeforeNonIterationStage=true
      NumAttributes=1
      NumCellsMaxAxis=10
      Precision=1000
      ResetValue=100000000
      SetResolutionMethod=Independent
      NumCells=<3 3 3>
      WorldBBoxSize=<100 100 100>
     }}
    Emitter.RasterizationGridBack=DataInterface() {propertyValues {
      CellSize=1
      ClearBeforeNonIterationStage=true
      NumAttributes=1
      NumCellsMaxAxis=10
      Precision=1000
      ResetValue=0
      SetResolutionMethod=Independent
      NumCells=<3 3 3>
      WorldBBoxSize=<100 100 100>
     }}
   ) {isSetParameters=true}
   /NiagaraFluids/Modules/Grid3D/RasterizationGrid3D_SetResolution(
    Grid:NiagaraDataInterfaceRasterizationGrid3D=$Emitter.RasterizationGrid
    NumCellsX:NiagaraInt32=/Game/Effects/FlipSimulation/Niagara/DynamicInputs/GetGrid2DSize(Grid2D Collection:NiagaraDataInterfaceGrid2DCollection=$Emitter.VelocityGrid, Y:NiagaraBool=false)
    NumCellsY:NiagaraInt32=&GetGrid2DSize(Grid2D Collection:NiagaraDataInterfaceGrid2DCollection=$Emitter.VelocityGrid, Y:NiagaraBool=true)
    NumCellsZ=1
    WorldGridExtents:Vector3f=<1 1 1>
   )
   &RasterizationGrid3D_SetResolution(
    Grid:NiagaraDataInterfaceRasterizationGrid3D=$Emitter.RasterizationGridBack
    NumCellsX:NiagaraInt32=&GetGrid2DSize(Grid2D Collection:NiagaraDataInterfaceGrid2DCollection=$Emitter.VelocityGrid, Y:NiagaraBool=false)
    NumCellsY:NiagaraInt32=&GetGrid2DSize(Grid2D Collection:NiagaraDataInterfaceGrid2DCollection=$Emitter.VelocityGrid, Y:NiagaraBool=true)
    NumCellsZ=1
    WorldGridExtents:Vector3f=<1 1 1>
   ) {module=RasterizationGrid3D_SetResolution001}
  ]
  emitterUpdateScript [
   /NiagaraFN/Modules/Scalability/SignificantInstancesCull(Cull Response:ENiagaraExecutionStateManagement=::KillImmediately)
   &EmitterState()
   /NiagaraFluids/Modules/Grid2D/Grid2D_UpdateSimAttributes(Grid:NiagaraDataInterfaceGrid2DCollection=$Emitter.VelocityGrid, WorldGridSize:Vector2f=$Emitter.Grid2D_SetResolution.WorldGridExtents) {module=UpdateSimAttributes}
   /NiagaraFluids/Modules/Grid2D/Grid2D_CreateUnitToWorldTransform(
    Camera:NiagaraDataInterfaceCamera=$Emitter.CameraQuery
    Camera Facing=true
    Local Pivot:Vector2f=<0.5 0.8>
    Offset:NiagaraPosition=$Engine.Owner.Position
    Rotation:Quat4f=&self:FN_PIC_FLIP_Fluid_Stream_Gas_1.NiagaraScratchPadContainer_0.ScratchDynamicInput_01_1(Rotation:Quat4f=$Engine.Owner.Rotation)
    WorldGridExtents:Vector2f=$Emitter.Grid2D_SetResolution.WorldGridExtents
   )
   /SetVariables(Emitter.Position:NiagaraPosition=&ConvertVectorToPosition(
     Input Position=$Emitter.Grid2D_CreateUnitToWorldTransform.LocalToWorldTranslation
     Conversion Operation:ENiagara_LWCConvertVecToPos=::NewEnumerator0
    )) {isSetParameters=true}
   /NiagaraFluids/Modules/Grid2D/Grid2D_SetBounds(
    EmitterProperties=DataInterface() {propertyValues {EmitterBinding {bindingMode=Self, emitterName=None}}}
    LocalPivot:Vector2f=$Emitter.Grid2D_CreateUnitToWorldTransform.LocalPivot
    WorldGridExtents:Vector2f=$Emitter.Grid2D_SetResolution.WorldGridExtents
   )
   /SetVariables(Emitter.OrthogonalVelocityMult=0.7) {isSetParameters=true}
   /NiagaraFluids/Modules/Grid2D/SplitMatrixIntoRows(Matrix:NiagaraMatrix=$Emitter.Grid2D_CreateUnitToWorldTransform.WorldToUnit)
   &SpawnRate(
    SpawnRate=&Multiply_Float(
     A=1500
     B=&FloatFromCurve(
      FloatCurve=DataInterface() {propertyValues {Curve {keys [
          <at 0 1>
          {arriveTangent=-0.2222, interpMode=RCIM_Cubic, leaveTangent=-0.2222, time=0.7, value=1}
          {arriveTangent=-3.3333, interpMode=RCIM_Cubic, leaveTangent=-3.3333, time=0.9, value=0.8}
          <at 1 0>
         ]}}}
      CurveIndex=$Engine.System.Age
     )
    )
    Use Spawn Probability=false
   )
   /Game/Effects/FlipSimulation/Niagara/Modules/Flip_SurfacingParameters(
    BaseColor_HighThickness:LinearColor=<rgba 0.0863 0.0353 0 1>
    BaseColor_LowThickness:LinearColor=<rgba 0.0706 0.0235 0 1>
    CustomBaseColor_HighThickness=true
    CustomBaseColor_lowThickness=true
    CustomDarkFaceBrightness=false
    CustomEdgeOffset=true
    CustomEnvReflection_Tint=true
    CustomFakeSSS_Color=false
    CustomHDRIReflection_Tint=false
    CustomHueShiftDarkFace=false
    CustomHueShiftLightFace=false
    CustomLightFaceBrightness=false
    CustomRefractionStrength=false
    CustomScaleBlurDistance=true
    CustomShadowStrength=false
    CustomThicknessDivideNumber=true
    CustomThicknessEmissive_HighThickness=true
    CustomThicknessEmissive_LowThickness=true
    CustomTotalOpacityScale=true
    DebugBound=true
    EnvReflection_Tint:LinearColor=<rgba 0.3281 0.0939 0 1>
    Preset:ENiagara_FlipRenderPresets=::NewEnumerator2
    ScaleBlurDistance=0.7
    ThicknessDivide=1
    ThicknessEmissive_HighThickness:LinearColor=<rgba 0.0863 0.0353 0 1>
    ThicknessEmissive_LowThickness:LinearColor=<rgba 0.0706 0.0235 0 1>
    TotalOpacityScale:NiagaraFloat=&FloatFromCurve(
     FloatCurve=DataInterface() {propertyValues {Curve {keys [{interpMode=RCIM_Cubic, tangentMode=RCTM_Break, time=0, value=0}, <at 0.5 1>]}}}
     CurveIndex=$System.Age
    )
   )
  ]
  particleSpawnScript [
   /Niagara/Modules/Spawn/Initialization/InitializeParticle(
    Color=<rgba 0 0.3383 1 1>
    Lifetime=&RandomRangeFloat(Minimum=1.2, Maximum=1.5)
    Module.Write Color=false
    Module.Write Lifetime=true
    Module.Write Mass=false
    Module.Write Material Random=true
    Module.Write Position=false
    Module.Write Scale=false
    Module.Write SpriteRotation=false
    Module.Write SpriteSize=true
    Position=&SimulationPosition()
    Sprite Size=<4 4>
   )
   /Game/Effects/FlipSimulation/Niagara/Modules/Flip_Source(
    DeltaTime:NiagaraFloat=$Emitter.DeltaTime
    EnableSimpleSourceMoving=true
    InheritVelocityScale=0
    Radius:NiagaraFloat=&FloatFromCurve(
     FloatCurve=DataInterface() {propertyValues {Curve {keys [{interpMode=RCIM_Cubic, tangentMode=RCTM_Break, time=0, value=0.3}, <at 0.6 1>]}}}
     CurveIndex=$System.Age
     Scale Curve=0.1
    )
    SourceAge=1
    SourceCount=8
    SourceMovingRemap=DataInterface() {propertyValues {Curve {keys [<at 0 0>,<at 1 1>]}}}
   )
   /Game/Effects/FlipSimulation/Niagara/Modules/Flip_SourceLinearMovement(LinearVelocityWS:Vector3f=<0 0 -5>)
   &CurlNoiseForce(
    Noise Frequency=500
    Noise Strength=&FloatFromCurve(
     FloatCurve=DataInterface() {propertyValues {Curve {keys [{interpMode=RCIM_Cubic, tangentMode=RCTM_Break, time=0, value=0}, {arriveTangent=0.1994, tangentMode=RCTM_User, time=0.9, value=1}]}}}
     CurveIndex=&Multiply_Float(A=$System.Age, B=1)
     Scale Curve=5
    )
    Pan Noise=true
    Pan Noise Field=<0.5 0.5 0.5>
   )
   &ShapeLocation(
    MaskSpawnGroup=false
    Override Seed=false
    Sphere Radius=&Multiply_Float(
     A=1.5
     B=&RemapRange(Input Value To Remap=$Particles.FlipSourceRandom, Input Min=0, Input Max=1, Output Min=0.5, Output Max=1)
    )
    UseOwnerScale=true
   )
   &AddVelocity(Velocity=<0 20 0>) {enabled=false}
   &ApplyInitialForces()
   /SetVariables(Particles.Velocity=&Multiply_Vector(
     A=$Particles.Velocity
     B=&VectorFromFloat(Value=&ClampFloat(Float=&Power(a=&Divide_Float(A=0.0167, B=$Engine.DeltaTime), b=0.5), Min=0.5, Max=2))
    )) {isSetParameters=true}
  ]
  particleUpdateScript [
   &ParticleState()
   &GravityForce(Gravity=<0 0 -500>)
   &PointForce(
    Force Origin=&AddVectorToPosition(Position=$Engine.Emitter.SimulationPosition, Vector=<0 0 0>)
    Force Strength=&FloatFromCurve(
     FloatCurve=DataInterface() {propertyValues {Curve {keys [{tangentMode=RCTM_User, time=0, value=0}, {tangentMode=RCTM_User, time=1, value=1}]}}}
     Scale Curve=80
    )
    Module.Use Falloff=true
    Module.UseLinearForce=false
   ) {enabled=false}
   &PointForce(
    Force Falloff Distance=70
    Force Origin=&AddVectorToPosition(Position=$Engine.Emitter.SimulationPosition, Vector=<0 0 -70>)
    Force Strength=-1500
    Module.Use Falloff=true
    Module.UseLinearForce=false
   ) {enabled=false, module=PointForce001}
   &Drag(Drag=0.2, Use Linear Drag=true, Use Rotational Drag=true)
   &CurlNoiseForce(
    Noise Frequency=150
    Noise Strength=&FloatFromCurve(
     FloatCurve=DataInterface() {propertyValues {Curve {keys [{tangentMode=RCTM_User, time=0.2, value=0}, {tangentMode=RCTM_User, time=1, value=1}]}}}
     Scale Curve=2000
    )
    Pan Noise=true
    Pan Noise Field=<0.5 0.5 0.5>
   ) {enabled=false, module=CurlNoiseForce001}
   /SetVariables(Transient.PhysicsForce:Vector3f=&Multiply_VectorByFloat(
     Vector=$Transient.PhysicsForce
     Float=&ClampFloat(
      Float=&Power(a=&Divide_Float(A=$Engine.DeltaTime, B=0.0167), b=0.5)
      Min=0.5
      Max=1.3
     )
    )) {isSetParameters=true}
   /SetVariables(Transient.StartPosition:NiagaraPosition=$Particles.Position) {isSetParameters=true}
   &SolveForcesAndVelocity(Clamp Velocity=false, DeltaTime:NiagaraFloat=$Emitter.DeltaTime, Limit Acceleration=false, Write to Presolve Properties=false)
   /SetVariables(Particles.Position:Vector3f=$Transient.StartPosition) {isSetParameters=true}
  ]
  rendererClasses [&NiagaraSpriteRendererProperties]
  renderers [
   {rendererClass=&NiagaraSpriteRendererProperties}
   {rendererClass=&NiagaraSpriteRendererProperties, rendererIndex=1}
   {rendererClass=&NiagaraSpriteRendererProperties, rendererIndex=2}
   {rendererClass=&NiagaraSpriteRendererProperties, rendererIndex=3}
   {rendererClass=&NiagaraSpriteRendererProperties, rendererIndex=4}
  ]
 }
 Gas_Ribbon {
  emitterUpdateScript [
   &EmitterState(Life Cycle Mode=::NewEnumerator1, Loop Delay=0.0388, Loop Duration=1, UseLoopDelay=false)
   &SpawnRate(
    SpawnRate=&Multiply_Float(
     A=20
     B=&FloatFromCurve(
      FloatCurve=DataInterface() {propertyValues {Curve {keys [<at 0 1 RCIM_Cubic>, {arriveTangent=-0.5241, tangentMode=RCTM_User, time=4, value=0}]}}}
      CurveIndex=$Engine.System.Age
     )
    )
    Use Spawn Probability=false
   )
  ]
  particleSpawnScript [
   /Niagara/Modules/Spawn/Initialization/V2/InitializeParticle(Position Offset=<0 -2 0>, Sprite Size Mode=::NewEnumerator3, Uniform Sprite Size=1, UsePositionOffset=true, Write Material Random=true) {module=InitializeParticle001}
   /SetVariables(Particles.RibbonWidth=30) {isSetParameters=true}
   &AddVelocity(Velocity=<0 30 0>) {enabled=false}
   &InheritVelocity(Inherited Velocity Amount Scale=<0.1 1 0>, Speed Limit=true, UseSpeedThreshold=false) {enabled=false}
   &CameraOffset(Camera Offset Amount=-5) {enabled=false}
  ]
  particleUpdateScript [
   &ParticleState()
   &Drag(Drag=0.9, Use Linear Drag=true, Use Rotational Drag=false)
   &GravityForce()
   &SolveForcesAndVelocity(Clamp Velocity=false, Limit Acceleration=false)
   &Material(Base Texture Time:NiagaraFloat=&Multiply_Float(A=$Emitter.Age, B=-1))
   &DynamicMaterialParameters(
    Index 0 Param 1=&MaskFloatBySpawnGroup(Masked Float=0, Passthrough Float=1)
    Index 0 Param 2=&FloatFromCurve(FloatCurve=DataInterface() {propertyValues {Curve {keys [{arriveTangent=-2.1032, interpMode=RCIM_Cubic, leaveTangent=-2.1032, tangentMode=RCTM_User, time=0, value=1}, <at 1 0.26>]}}})
    Module.Param0WriteEnabled=true
    Module.Param1WriteEnabled=true
    Module.Param2WriteEnabled=false
    Module.Param3WriteEnabled=false
   )
   &ScaleRibbonWidth(Ribbon Width Scale=&FloatFromCurve(
     FloatCurve=DataInterface() {propertyValues {Curve {keys [
         {arriveTangent=21.7132, interpMode=RCIM_Cubic, leaveTangent=21.7132, tangentMode=RCTM_User, time=0, value=0}
         <at 0.05 0.8 RCIM_Cubic>
         {arriveTangent=0.0983, tangentMode=RCTM_User, time=0.95, value=0}
        ]}}}
     CurveIndex=$Emitter.Age
    ))
   /SetVariables(Particles.RibbonLinkOrder:NiagaraFloat=$Particles.NormalizedAge) {isSetParameters=true}
  ]
  rendererClasses [&NiagaraRibbonRendererProperties]
  renderers [{rendererClass=&NiagaraRibbonRendererProperties}]
 }
}
summary {
 systemName=NS_GasCan_Pour_Fluid
 emitters [
  {emitterName=Drips_CPU_Low, rendererClasses [&NiagaraSpriteRendererProperties]}
  {emitterName=Drips_GPU, simTarget=GPUComputeSim, rendererClasses [&NiagaraSpriteRendererProperties]}
  {
   bEnabled=false
   emitterName=FN_PIC_FLIP_Fluid_Stream_Gas
   simTarget=GPUComputeSim
   rendererClasses [&NiagaraSpriteRendererProperties]
  }
  {emitterName=BakeFluid, rendererClasses [&NiagaraSpriteRendererProperties]}
  {emitterName=Gas_Ribbon, rendererClasses [&NiagaraRibbonRendererProperties]}
 ]
 userVariables []
}
